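# Material function for the Donut Plate female character body that animates hair with a sine wave added to the material's World Position Offset.
# Vertex red drives wave phase and squared falloff, while vertex blue varies time offset, speed and frequency.
# The remapped sine scales a direction vector parameter.
# A normalized pre-skinned position, minus an adjustment vector, is also added to the offset without vertex color masking.
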MaterialFunction /Game/Characters/Player/Female/Medium/Bodies/F_MED_Donut_Plate/Materials/MF_DonutHairAnimation
graph {nodes [
  n1 = GetMaterialAttributes(None=FunctionInput() {inputName=Material, inputType=FunctionInput_MaterialAttributes}) {attributeGetTypes [#1]}
  n2 = VertexColor()
  n3 = Multiply(A=@n2.B, B=ScalarParameter() {defaultValue=1, group=DonutHair, parameterName=DH_Offset_Amt})
  n4 = Time()
  n5 = LinearInterpolate(
   A=ScalarParameter() {defaultValue=0.25, group=DonutHair, parameterName=DH_MinWaveSpeed}
   B=ScalarParameter() {defaultValue=0.25, group=DonutHair, parameterName=DH_MaxWaveSpeed}
   Alpha=@n2.B
  )
  n6 = ScalarParameter() {defaultValue=1, group=DonutHair, parameterName=DH_GlobalWaveSpeed}
  n7 = Subtract(A=@n2.R, B=Multiply(A=Add(A=@n3, B=@n4), B=Multiply(A=@n5, B=@n6)))
  n8 = LinearInterpolate(
   A=ScalarParameter() {defaultValue=2, group=DonutHair, parameterName=DH_MinSineWave}
   B=ScalarParameter() {defaultValue=1, group=DonutHair, parameterName=DH_MaxSineWave}
   Alpha=@n2.B
  )
  n9 = ConstantBiasScale(None=Sine(None=Multiply(A=@n7, B=@n8)))
  n10 = VectorParameter() {group=DonutHair, parameterName=DH_OffsetAmtAndDirection, defaultValue=<rgba 2 1 1 1>}
  n11 = PreSkinnedPosition()
  n12 = Add(
   A=Subtract(A=Normalize(VectorInput=@n11), B=VectorParameter() {group=DonutHair, parameterName=DH_Adjustments, defaultValue=<rgba 0 0 1 1>}.RGB)
   B=Multiply(A=Multiply(A=@n2.R, B=@n9), B=Multiply(A=@n2.R, B=@n10.RGB))
  )
  FunctionOutput(None=SetMaterialAttributes!(MaterialAttributes=@n1.MaterialAttributes, WorldPositionOffset=Add(A=@n1.WorldPositionOffset, B=@n12)) {attributeSetTypes [#1]})
 ]}
