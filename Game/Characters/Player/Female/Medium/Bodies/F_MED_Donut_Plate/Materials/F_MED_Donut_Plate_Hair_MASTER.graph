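# Master material for the hair of the medium female Donut Plate player character, a skeletal mesh with morph target and deformer support.
# Material attributes flow from the shared character PreFX function through Silk, switchable Cloth Fuzz and Clear Coat layers, each masked by SkinFX_Mask channels.
# MF_DonutHairAnimation then processes the result before the shared PostFX function.
# PreFX's AO output also feeds the PostFX Gmap, Fresnel and Distance Blend inputs.
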Material /Game/Characters/Player/Female/Medium/Bodies/F_MED_Donut_Plate/Materials/F_MED_Donut_Plate_Hair_MASTER
graph {
 properties {
  bUseMaterialAttributes=true
  bUsedWithMeshDeformer=true
  bUsedWithMorphTargets=true
  bUsedWithSkeletalMesh=true
  pixelDepthOffsetMode=PDOM_Legacy
  usedShadingModels=Default Lit
 }
 nodes [
  n1 = MaterialFunctionCall() {materialFunction=&BaseFN_CharacterFunction_PreFX}
  n2 = TextureSampleParameter2D() {parameterName=SkinFX_Mask}
  n3 = MaterialFunctionCall!(Alpha=ChannelMaskParameter(None=@n2.RGBA) {group=Skin FX, parameterName=Silk Channel}.R, Material=@n1.Result) {materialFunction=/Game/Weapons/WeaponSkins/MaterialFunction/MF_Silk}
  n4 = StaticSwitchParameter(
   True=MaterialFunctionCall!(
    Alpha=StaticComponentMaskParameter(None=@n2.RGBA) {defaultR=true, group=Skin FX, parameterName=Cloth Channel}
    Gmap=@n1.Gmap
    Material=@n3.Result
   ) {materialFunction=/Game/Weapons/WeaponSkins/MaterialFunction/ClothFuzz}.Result
   False=@n3.Result
  ) {group=Skin FX, parameterName=Use Cloth fuzz}
  n5 = MaterialFunctionCall!(Alpha=StaticComponentMaskParameter(None=@n2.RGBA) {defaultR=true, group=Skin FX, parameterName=Clear Coat Channel}, In=@n4) {materialFunction=/Game/Weapons/WeaponSkins/MaterialFunction/Clearcoat}
  MaterialFunctionCall(
   AO=@n1.AO
   Distance Blend=@n1.AO
   Fresnel=@n1.AO
   Gmap=@n1.AO
   Material=MaterialFunctionCall!(Material=StaticSwitchParameter(True=@n5.Result, False=@n4) {group=Skin FX, parameterName=Use Clear Coat}) {materialFunction=&/MF_DonutHairAnimation}.Result
  ) {materialFunction=&BaseFN_CharacterFunction_PostFX}
 ]
}
